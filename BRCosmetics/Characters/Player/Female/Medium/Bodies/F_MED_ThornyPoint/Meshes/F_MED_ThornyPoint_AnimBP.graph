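# Animation Blueprint for the F_MED_ThornyPoint female body cosmetic, driving its accessory bones on top of a pose copied from the attached parent mesh.
# Two weighted multi-target constraints keep the front padlock base and knives base following torso bones.
# Anim Dynamics then swings the padlock bone with angular spring, heavy damping, custom gravity and limits.
# The ThornyPoint Control Rig runs last before output, and the event graph holds no logic.
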AnimBlueprint /BRCosmetics/Characters/Player/Female/Medium/Bodies/F_MED_ThornyPoint/Meshes/F_MED_ThornyPoint_AnimBP
target_skeleton=&/F_MED_ThornyPoint_Skeleton
graphs [
 AnimGraph {nodes [
   n1 = "Constraint - dyn_padlock_fr_base (PPPPPPPP)"!(
    ComponentPose=Local To Component(LocalPose=Copy Pose From Mesh() {bUseAttachedParent=true})
    Alpha=1
    ConstraintWeights_0=0.3253
    ConstraintWeights_1:float=0
    ConstraintWeights_2:float=0.0001
    ConstraintWeights_3:float=0.6672
    ConstraintWeights_4:float=0
    ConstraintWeights_5:float=0
    ConstraintWeights_6:float=0.0034
    ConstraintWeights_7:float=0.004
   ) -> (Pose) {
    boneToModify {boneName=dyn_padlock_fr_base}
    constraintSetup [
     {targetBone {boneName=spine_05}}
     {targetBone {boneName=clavicle_l}}
     {targetBone {boneName=clavicle_r}}
     {targetBone {boneName=spine_04}}
     {targetBone {boneName=spine_03}}
     {targetBone {boneName=spine_02}}
     {targetBone {boneName=deform_pec_l}}
     {targetBone {boneName=deform_pec_r}}
    ]
    constraintWeights [1,1,1,1,1,1,1,1]
   }
   n2 = "Constraint - base_knives1 (PPPPPPPP)"!(
    ComponentPose=@n1
    Alpha=1
    ConstraintWeights_0=0.0018
    ConstraintWeights_1:float=0.0018
    ConstraintWeights_2:float=0.0754
    ConstraintWeights_3:float=0.06
    ConstraintWeights_4:float=0.3641
    ConstraintWeights_5:float=0.0955
    ConstraintWeights_6:float=0.4001
    ConstraintWeights_7:float=0.0012
   ) -> (Pose) {
    boneToModify {boneName=base_knives1}
    constraintSetup [
     {targetBone {boneName=spine_05}}
     {targetBone {boneName=clavicle_r}}
     {targetBone {boneName=spine_04}}
     {targetBone {boneName=spine_03}}
     {targetBone {boneName=spine_02}}
     {targetBone {boneName=spine_01}}
     {targetBone {boneName=pelvis}}
     {targetBone {boneName=deform_glute_r}}
    ]
    constraintWeights [1,1,1,1,1,1,1,1]
   }
   n3 = Anim Dynamics - Bone: dyn_padlock_fr_01(ComponentPose=@n2, Alpha=1, ExternalForce="0,0,0") {
    angularDampingOverride=0.999
    angularSpringConstant=199
    bAngularSpring=true
    bOverrideAngularDamping=true
    bUseGravityOverride=true
    alphaBoolBlend {blendOption=Linear}
    boundBone {boneName=dyn_padlock_fr_01}
    componentAppliedLinearAccClamp=<5000 5000 5000>
    componentLinearAccScale=<0.3 0.3 0.3>
    componentLinearVelScale=<0.3 0.3 0.3>
    gravityOverride {z=-700}
    physicsBodyDefinitions [{
      boundBone {boneName=dyn_padlock_fr_01}
      boxExtents=<8 8 8>
      constraintSetup {angularLimitsMax=<3 1 3>, angularLimitsMin=<-3 -70 -3>, angularTarget=<1 0 0>}
      localJointOffset=<0 0 1>
     }]
    relativeSpaceBone {boneName=spine_05}
   }
   n4 = Control Rig(Source=Component To Local(ComponentPose=@n3)) {controlRigAssetReference {blueprintRigClass=&/F_MED_ThornyPoint_CtrlRig.F_MED_ThornyPoint_CtrlRig_C}}
   Output Pose(Result=@n4) {layerGroup=None}
  ]}
 EventGraph {nodes [Event BlueprintUpdateAnimation()]}
]
variables [Custom Base Mesh {type=SkeletalMeshComponent}]
