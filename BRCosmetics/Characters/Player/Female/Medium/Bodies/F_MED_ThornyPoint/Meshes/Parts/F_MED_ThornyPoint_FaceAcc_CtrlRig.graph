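# Control Rig for the F_MED_ThornyPoint face accessory part, applying corrective offsets to dynamic hair, chain and glasses bones.
# A shared DeformRig function runs first, then a sequence offsets hair and chain bones by head rotation relative to spine_05.
# A second correction multiplies that head weight by a spherical pose reader on the right upper arm.
# Inner forehead bone movement drives the glasses offset through a clamped sum.
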ControlRigBlueprint /BRCosmetics/Characters/Player/Female/Medium/Bodies/F_MED_ThornyPoint/Meshes/Parts/F_MED_ThornyPoint_FaceAcc_CtrlRig
graphs [
 RigVMFunctionLibrary {links [], nodes []}
 RigVMFunction_Sequence_1_ContainedGraph {nodes [
   n1 = RigVMFunctionEntryNode()
   RigVMFunction_Sequence = &RigVMFunction_Sequence(ExecuteContext=@n1.ExecuteContext)
   n2 = &RigVMFunction_Sequence(ExecuteContext=@RigVMFunction_Sequence.B)
   RigVMFunctionReturnNode(A=@RigVMFunction_Sequence.A, B=@n2.A, C=@n2.B)
  ]}
 RigVMModel {nodes [
   RigUnit_BeginExecution = &RigUnit_BeginExecution()
   FN_DeformRig_FNC = RigVMFunctionReferenceNode(
    DeformRig_ Curve="(deformRig_l_upperarm_up_alpha_72_#1=true,deformRig_r_upperarm_up_alpha_73_#2=true,deformRig_l_upperarm_fr_alpha_74_#3=true,deformRig_r_upperarm_fr_alpha_75_CB51B80C4 ...(truncated)"
    DeformRig_Bool="(upperarm_1_#4=true,elbow_3_#5=false,wrist_7_#6=false,groin_8_#7=false,glute_17_#8=false,knee_9_40C408324C ...(truncated)"
    DeformRig_Debug=(upperarm_debug_18_#4=False,wrist_debug_20_#6=False,lat_debug_26_#9=False,groinGlute_debug_27_#7=False)
    DeformRig_Value="(upperarm_up_18_#10=(Rotation=(X=0,Y=0,Z=0,W=1),Translation=(X=0,Y=0,Z=8),Scale3D=(X=0.7,Y=1,Z=1)),upperarm_fwd_fr_19_#11=(Rotation=(X=0,Y=0,Z=0,W=1),Translation=(X=0,Y=10,Z=0),Scale3D=(X=0.5,Y=1,Z= ...(truncated)"
    ExecuteContext=@RigUnit_BeginExecution.ExecutePin
   )
   > RigVMFunction_Sequence = RigVMAggregateNode()
   n1 = &RigVMFunction_MathDoubleRemap(
    SourceMaximum=-3
    SourceMinimum=-1
    TargetMaximum=1
    TargetMinimum=0
    Value=&RigUnit_GetTransform(Item=(Type=Bone,Name="FACIAL_L_ForeheadIn"), Space=LocalSpace, bInitial=False).Transform.Translation.Y
   )
   n2 = &RigVMFunction_MathDoubleRemap(
    SourceMaximum=-3
    SourceMinimum=-1
    TargetMaximum=1
    TargetMinimum=0
    Value=&RigUnit_GetTransform(Item=(Type=Bone,Name="FACIAL_R_ForeheadIn"), Space=LocalSpace, bInitial=False).Transform.Translation.Y
   )
   &RigUnit_ModifyTransforms(
    ItemToModify=((Item=(Type=Bone,Name="dyn__glasses_01"),Transform=(Rotation=(X=0.0003,Y=-0.0175,Z=0.0175,W=0.9997),Translation=(X=0,Y=0,Z=0),Scale3D=(X=1,Y=1,Z=1))))
    WeightMinimum=0
    Weight=&RigVMFunction_MathDoubleClamp(Value=&RigVMFunction_MathDoubleAdd(A=@n1.Result, B=@n2.Result).Result).Result
    ExecutePin=@RigVMFunction_Sequence.C
   )
   n3 = &RigUnit_SphericalPoseReader(
    ActiveRegionScaleFactors=(PositiveWidth=1,NegativeWidth=1,PositiveHeight=1,NegativeHeight=1)
    ActiveRegionSize=0.1
    Debug=(bDrawDebug=false,bDraw2D=False,bDrawLocalAxes=False,DebugScale=25,DebugSegments=20,DebugThickness=0.25)
    DriverAxis=(X=-1,Y=0,Z=0)
    DriverItem=(Type=Bone,Name="upperarm_r")
    FalloffRegionScaleFactors=(PositiveWidth=1,NegativeWidth=1,PositiveHeight=1,NegativeHeight=1)
    FalloffSize=0.2
    FlipHeightScaling=False
    FlipWidthScaling=False
    OptionalParentItem=(Type=Bone,Name="None")
    RotationOffset=(X=-120,Y=55,Z=33)
    ExecutePin=@RigVMFunction_Sequence.B
   )
   n4 = &RigVMFunction_MathQuaternionToEuler(
    RotationOrder=ZYX
    Value=&RigUnit_GetRelativeTransformForItem(Child=(Type=Bone,Name="head"), Parent=(Type=Bone,Name="spine_05"), bChildInitial=False, bParentInitial=False).RelativeTransform.Rotation
   )
   > &RigUnit_ModifyTransforms(
    ItemToModify="((Item=(Type=Bone,Name=\"dyn_hair_bk_l_base\"),Transform=(Rotation=(X=0,Y=-0.0958,Z=0,W=0.9954),Translation=(X=-2,Y=0,Z=0),Scale3D=(X=1,Y=1,Z=1))),(Item=(Type=Bone,Name=\"dyn_chain_bk_r_base\"),Transform=(Rotation=(X=0,Y=0,Z=0.0175,W=0.9998),Translation=(X=-2, ...(truncated)"
    WeightMinimum=0
    Weight=&RigVMFunction_MathDoubleMul(
     B=&RigVMFunction_MathDoubleRemap(SourceMaximum=-45, SourceMinimum=0, TargetMaximum=1, TargetMinimum=0, Value=@n4.Result.Z).Result
     A=@n3.OutputParam
    ).Result
   )
   RigVMFunction_MathQuaternionToEuler = &RigVMFunction_MathQuaternionToEuler(
    RotationOrder=ZYX
    Value=&RigUnit_GetRelativeTransformForItem(Child=(Type=Bone,Name="head"), Parent=(Type=Bone,Name="spine_05"), bChildInitial=False, bParentInitial=False).RelativeTransform.Rotation
   )
   &RigUnit_ModifyTransforms(
    ItemToModify="((Item=(Type=Bone,Name=\"dyn_hair_bk_l_base\"),Transform=(Rotation=(X=-0.0183,Y=-0.1899,Z=-0.0941,W=0.9771),Translation=(X=-1,Y=0,Z=0),Scale3D=(X=1,Y=1,Z=1))),(Item=(Type=Bone,Name=\"dyn_chain_bk_r_base\"),Transform=(Rotation=(X=0.0364,Y=-0.1873,Z=0.1873,W=0.9 ...(truncated)"
    WeightMinimum=0
    Weight=&RigVMFunction_MathDoubleRemap(SourceMaximum=-45, SourceMinimum=0, TargetMaximum=1, TargetMinimum=0, Value=@RigVMFunction_MathQuaternionToEuler.Result.Z).Result
    ExecutePin=@RigVMFunction_Sequence.A
   )
  ]}
]
hierarchy [
 root {loc=<0 0 0>, rot=<0 0 0 1>}
 pelvis {parent=root, loc=<0 -0.5301 98.0571>, rot=<0.0127 0.707 -0.0127 -0.707>}
 spine_01 {parent=pelvis, loc=<6.1208 -0.9328 0>, rot=<0 0 0.034 -0.9994>}
 spine_02 {parent=spine_01, loc=<6.5997 -0.2095 0>, rot=<0 0 0.0238 -0.9997>}
 spine_03 {parent=spine_02, loc=<6.4914 -0.1237 0>, rot=<0 0 0.0149 -0.9999>}
 spine_04 {parent=spine_03, loc=<6.9598 0.0794 0>, rot=<0 0 -0.0132 -0.9999>}
 spine_05 {parent=spine_04, loc=<6.9987 0 0>, rot=<0 0 -0.0015 -1>}
 neck_01 {parent=spine_05, loc=<11.3138 0.1529 0>, rot=<0 0 -0.2003 -0.9797>}
 neck_02 {parent=neck_01, loc=<4.5627 0 0>, rot=<0 0 0.0504 -0.9987>}
 head {parent=neck_02, loc=<5.3914 0 0>, rot=<0 0 0.115 -0.9934>}
 dyn_hair_fr_r_base2 {parent=head, loc=<0.7189 8.09 5.0927>, rot=<0.0931 -0.9937 0.0079 0.0614>}
 dyn_hair_fr_r_02 {parent=dyn_hair_fr_r_base2, loc=<0 0 0>, rot=<0 0 0 1>}
 dyn_hair_fr_r_03 {parent=dyn_hair_fr_r_02, loc=<5.916 0 0>, rot=<0.0047 -0.1166 0.0664 -0.9909>}
 dyn_hair_fr_r_04 {parent=dyn_hair_fr_r_03, loc=<3.7901 0 0>, rot=<0 0 0 1>}
 dyn_chain_bk_r_base {parent=head, loc=<7.856 -4.0522 5.5941>, rot=<0.1151 -0.2594 0.9577 -0.0472>}
 dyn_chain_bk_r_01 {parent=dyn_chain_bk_r_base, loc=<0 0 0>, rot=<0 -0.0339 0 -0.9994>}
 dyn_chain_bk_r_02 {parent=dyn_chain_bk_r_01, loc=<7.484 0 0.1276>, rot=<0 0.0054 0 -1>}
 dyn_chain_bk_r_03 {parent=dyn_chain_bk_r_02, loc=<6.6227 0 -0.067>, rot=<0 0 0 1>}
 dyn_chain_bk_r_04 {parent=dyn_chain_bk_r_03, loc=<5 0 0>, rot=<0 0 0 1>}
 dyn_chain_back_base {parent=head, loc=<8.1525 -6.1489 -0.3109>, rot=<0.0708 -0.6895 0.7177 -0.0662>}
 dyn_chain_back_01 {parent=dyn_chain_back_base, loc=<0 0 0>, rot=<0 -0.0102 0 -0.9999>}
 dyn_chain_back_02 {parent=dyn_chain_back_01, loc=<7.1245 0 -0.1868>, rot=<0 -0.0055 0 -1>}
 dyn_chain_back_03 {parent=dyn_chain_back_02, loc=<6.7491 0 0>, rot=<0 0 0 1>}
 dyn_chain_back_04 {parent=dyn_chain_back_03, loc=<5 0 0>, rot=<0 0 0 1>}
 dyn_hair_l_base {parent=head, loc=<1.5477 1.8069 -7.2911>, rot=<-0.0126 -0.9996 -0.0005 -0.0261>}
 dyn_hair_l_01 {parent=dyn_hair_l_base, loc=<0 0 0>, rot=<0 0.0018 0.0269 -0.9996>}
 dyn_hair_l_02 {parent=dyn_hair_l_01, loc=<8.069 0 0>, rot=<0 -0.0084 0 -1>}
 dyn_hair_l_03 {parent=dyn_hair_l_02, loc=<3.753 0 0>, rot=<0 0 0 1>}
 dyn_hair_l_04 {parent=dyn_hair_l_03, loc=<6.4578 0 0>, rot=<0 0 0 1>}
 dyn__glasses_base {parent=head, loc=<2.7261 0.6802 0.0046>, rot=<0.0041 -1 0 0>}
 dyn__glasses_01 {parent=dyn__glasses_base, loc=<0 0 0>, rot=<0 0 0 1>}
 dyn_hair_bk_l_base {parent=head, loc=<7.8728 -2.3561 -6.6692>, rot=<0.0649 -0.9606 0.2602 -0.0725>}
 dyn_hair_bk_l_01 {parent=dyn_hair_bk_l_base, loc=<0 0 0>, rot=<0 -0.0144 0 -0.9999>}
 dyn_hair_bk_l_02 {parent=dyn_hair_bk_l_01, loc=<7.4242 0 0.1074>, rot=<0 0.0146 0 -0.9999>}
 dyn_hair_bk_l_03 {parent=dyn_hair_bk_l_02, loc=<6.496 0 0>, rot=<0 0 0 1>}
 dyn_hair_bk_l_04 {parent=dyn_hair_bk_l_03, loc=<5 0 0>, rot=<0 0 0 1>}
 dyn_hair_bk_base {parent=head, loc=<5.699 -6.8696 -0.2429>, rot=<0.027 -0.7068 0.7066 -0.0184>}
 dyn_hair_bk_04 {parent=dyn_hair_bk_base, loc=<0 0 0>, rot=<0 0 0 1>}
 dyn_hair_bk_02 {parent=dyn_hair_bk_04, loc=<6.6064 0 0>, rot=<0 0.073 0 -0.9973>}
 dyn_hair_bk_03 {parent=dyn_hair_bk_02, loc=<6.4485 0 0>, rot=<0 0 0 1>}
 dyn_hair_bk_r_base {parent=head, loc=<4.6646 -0.4275 7.1727>, rot=<0.035 -0.0004 0.9992 -0.0194>}
 dyn_hair_bk_r_01 {parent=dyn_hair_bk_r_base, loc=<0 0 0>, rot=<0 0 0 1>}
 dyn_hair_bk_r_02 {parent=dyn_hair_bk_r_01, loc=<5.8366 0 0>, rot=<0 0.0223 0 -0.9998>}
 dyn_hair_bk_r_03 {parent=dyn_hair_bk_r_02, loc=<6.787 0 0>, rot=<0 0 0 1>}
 dyn_hair_fr_l_base2 {parent=head, loc=<0.1144 6.8997 -6.044>, rot=<-0.0202 -0.9943 -0.0251 -0.102>}
 dyn_hair_fr_l_05 {parent=dyn_hair_fr_l_base2, loc=<0 0 0>, rot=<0 0.0252 0 -0.9997>}
 dyn_hair_fr_l_03 {parent=dyn_hair_fr_l_05, loc=<5.6382 0 0>, rot=<-0.0052 -0.1183 0.0333 -0.9924>}
 dyn_hair_fr_l_04 {parent=dyn_hair_fr_l_03, loc=<3.9247 0 0>, rot=<0 0 0 1>}
 dyn_chain_r_base {parent=head, loc=<1.6094 1.8972 7.3802>, rot=<0.0291 -0.0317 0.999 -0.0097>}
 dyn_chain_r_01 {parent=dyn_chain_r_base, loc=<0 0 0>, rot=<0 0 0 1>}
 dyn_chain_r_02 {parent=dyn_chain_r_01, loc=<8.3592 0 0.4576>, rot=<0 -0.0947 0 -0.9955>}
 dyn_chain_r_03 {parent=dyn_chain_r_02, loc=<3.7434 0 -0.0183>, rot=<0 0.0108 0 -0.9999>}
 dyn_chain_r_04 {parent=dyn_chain_r_03, loc=<5.3293 0 0>, rot=<0 0 0 1>}
 FACIAL_C_FacialRoot {parent=head, loc=<2.1524 -0.746 -0.0002>, rot=<0.4879 -0.5118 -0.4879 -0.5118>}
 FACIAL_C_Forehead {parent=FACIAL_C_FacialRoot, loc=<0.0013 -2.0521 9.0707>, rot=<-0.0451 -0.0004 0.0003 -0.999>}
 FACIAL_L_ForeheadIn {parent=FACIAL_C_FacialRoot, loc=<1.6675 -2.1255 8.8919>, rot=<-0.0624 -0.07 0.0045 -0.9956>}
 FACIAL_R_ForeheadIn {parent=FACIAL_C_FacialRoot, loc=<-1.6542 -2.1255 8.8995>, rot=<-0.0624 0.0693 -0.0038 -0.9956>}
 FACIAL_L_ForeheadMid {parent=FACIAL_C_FacialRoot, loc=<3.1048 -2.5424 8.3319>, rot=<-0.0727 -0.176 -0.0206 -0.9815>}
 FACIAL_R_ForeheadMid {parent=FACIAL_C_FacialRoot, loc=<-3.095 -2.5453 8.3356>, rot=<-0.0726 0.1753 0.0212 -0.9816>}
 FACIAL_L_ForeheadOut {parent=FACIAL_C_FacialRoot, loc=<4.2819 -2.5925 7.3583>, rot=<-0.1231 -0.2558 -0.0793 -0.9556>}
 FACIAL_R_ForeheadOut {parent=FACIAL_C_FacialRoot, loc=<-4.2781 -2.5935 7.3663>, rot=<-0.1229 0.2551 0.0799 -0.9557>}
 FACIAL_C_Nose {parent=FACIAL_C_FacialRoot, loc=<0.0007 1.4061 9.3943>, rot=<-0.132 -0.0003 0.0003 -0.9912>}
 FACIAL_C_NoseLower {parent=FACIAL_C_Nose, loc=<-0.0035 2.602 -0.3252>, rot=<0.1305 0 0 -0.9914>}
 FACIAL_L_Nostril {parent=FACIAL_C_Nose, loc=<0.8494 2.0392 -0.4046>, rot=<0.0173 0.1305 0.0023 -0.9913>}
 FACIAL_R_Nostril {parent=FACIAL_C_Nose, loc=<-0.8537 2.0412 -0.4296>, rot=<0.0173 -0.1305 -0.0023 -0.9913>}
 FACIAL_C_NoseTip {parent=FACIAL_C_Nose, loc=<0 1.899 0.4409>, rot=<0.0872 0 0 -0.9962>}
 clavicle_l {parent=spine_05, loc=<5.4831 -0.1053 -3.2946>, rot=<0.0465 -0.7289 -0.009 -0.6829>}
 upperarm_l {parent=clavicle_l, loc=<9.8658 -1.2136 -0.4105>, rot=<0.0164 -0.3245 -0.0269 -0.9454>}
 lowerarm_l {parent=upperarm_l, loc=<26.465 -0.8345 0.259>, rot=<0.0036 0.0161 -0.2298 -0.9731>}
 hand_l {parent=lowerarm_l, loc=<22.7355 -0.3203 -0.6105>, rot=<-0.6784 -0.0223 0.0735 -0.7307>}
 thumb_01_l {parent=hand_l, loc=<1.3118 -0.3191 -2.459>, rot=<0.6546 -0.4042 0.0664 -0.6354>}
 thumb_02_l {parent=thumb_01_l, loc=<5.9336 -0.0448 0.0401>, rot=<-0.0037 0.0213 0.298 -0.9543>}
 thumb_03_l {parent=thumb_02_l, loc=<2.2223 0.3682 -0.0907>, rot=<0.0118 -0.013 0.0066 -0.9998>}
 index_metacarpal_l {parent=hand_l, loc=<1.6535 0.0183 -2.96>, rot=<0.0158 -0.0124 -0.0173 -0.9996>}
 index_01_l {parent=index_metacarpal_l, loc=<8.3008 0.0357 -0.0524>, rot=<-0.0044 0.0485 0.0924 -0.9945>}
 index_02_l {parent=index_01_l, loc=<3.3587 -0.3004 0.0452>, rot=<-0.002 0.0027 0.3225 -0.9466>}
 index_03_l {parent=index_02_l, loc=<2.5578 -0.123 -0.0252>, rot=<-0.002 0.0067 0.2558 -0.9667>}
 middle_metacarpal_l {parent=hand_l, loc=<2.1856 -0.1725 -0.5984>, rot=<0.006 0.0168 -0.0329 -0.9993>}
 middle_01_l {parent=middle_metacarpal_l, loc=<7.3275 -0.2447 0.0129>, rot=<-0.0208 0.0042 0.0858 -0.9961>}
 middle_02_l {parent=middle_01_l, loc=<4.0123 0.1401 0.1175>, rot=<-0.0218 0.0258 0.3512 -0.9357>}
 middle_03_l {parent=middle_02_l, loc=<2.6077 0.0534 0.1373>, rot=<0.0082 0.0188 0.1708 -0.9851>}
 ring_metacarpal_l {parent=hand_l, loc=<2.2074 -0.3625 1.3638>, rot=<-0.0284 0.0722 -0.0458 -0.9959>}
 ring_01_l {parent=ring_metacarpal_l, loc=<7.0295 -0.6197 -0.1446>, rot=<-0.0489 -0.0252 0.1892 -0.9804>}
 ring_02_l {parent=ring_01_l, loc=<3.6174 0.4647 -0.1027>, rot=<0 0 0.2806 -0.9598>}
 ring_03_l {parent=ring_02_l, loc=<2.4152 0.114 0.0199>, rot=<0 0 0.1189 -0.9929>}
 pinky_metacarpal_l {parent=hand_l, loc=<1.7616 -0.9753 3.3814>, rot=<-0.0397 0.0814 -0.069 -0.9935>}
 pinky_01_l {parent=pinky_metacarpal_l, loc=<7.1709 -0.4831 -0.1094>, rot=<-0.0654 -0.0035 0.2242 -0.9723>}
 pinky_02_l {parent=pinky_01_l, loc=<2.8274 0.0893 -0.1043>, rot=<0.0037 0.0011 0.2854 -0.9584>}
 pinky_03_l {parent=pinky_02_l, loc=<2.2046 0.0307 -0.0078>, rot=<-0.0002 0.0098 0.1445 -0.9895>}
 deform_wrist_dn_base_l {parent=hand_l, loc=<0 -1.5326 0>, rot=<0 0 0 -1>}
 deform_wrist_dn_l {parent=deform_wrist_dn_base_l, loc=<0 0 0>, rot=<0 0 0 1>}
 deform_wrist_up_base_l {parent=hand_l, loc=<0 2.3009 0>, rot=<0 0 0 -1>}
 deform_wrist_up_l {parent=deform_wrist_up_base_l, loc=<0 0 0>, rot=<0 0 0 1>}
 lowerarm_twist_02_l {parent=lowerarm_l, loc=<10.6707 0 0>, rot=<0 0 0 1>}
 lowerarm_twist_01_l {parent=lowerarm_l, loc=<18.316 0 0>, rot=<0 0 0 1>}
 deform_elbow_in_base_l {parent=lowerarm_l, loc=<0.1834 3.6062 -0.2572>, rot=<0 0 0 -1>}
 deform_elbow_in_l {parent=deform_elbow_in_base_l, loc=<0 0 0>, rot=<0 0 0 1>}
 deform_elbow_out_base_l {parent=lowerarm_l, loc=<0 -4 0>, rot=<0 0 0 1>}
 deform_elbow_out_l {parent=deform_elbow_out_base_l, loc=<0 0 0>, rot=<0 0 0 1>}
 upperarm_twist_01_l {parent=upperarm_l, loc=<7.2135 -0.1134 0>, rot=<0 0 0 1>}
 deform_upperarm_bk_base_l {parent=upperarm_twist_01_l, loc=<-6.4053 -4.3746 0.314>, rot=<0 0 0 -1>}
 deform_upperarm_bk_l {parent=deform_upperarm_bk_base_l, loc=<0 0 0>, rot=<0 0 0 1>}
 deform_upperarm_fr_base_l {parent=upperarm_twist_01_l, loc=<-5.6312 5.092 0.3354>, rot=<0 0 0 -1>}
 deform_upperarm_fr_l {parent=deform_upperarm_fr_base_l, loc=<0 0 0>, rot=<0 0 0 1>}
 deform_upperarm_up_base_l {parent=upperarm_twist_01_l, loc=<-8.891 0.4825 3.9517>, rot=<0 0 0 -1>}
 deform_upperarm_up_l {parent=deform_upperarm_up_base_l, loc=<0 0 0>, rot=<0 0 0 1>}
 upperarm_twist_02_l {parent=upperarm_l, loc=<19.7998 -0.1134 0>, rot=<0 0 0 1>}
 dyn_thorn_2_l_base {parent=upperarm_l, loc=<-0.2975 0.1824 4.0373>, rot=<-0.0074 -0.4087 -0.0019 -0.9126>}
 dyn_thorn_2_l_01 {parent=dyn_thorn_2_l_base, loc=<0 0 0>, rot=<0 0 0 1>}
 dyn_strap_l_base {parent=clavicle_l, loc=<5.7826 -1.1776 4.3803>, rot=<0 0 0 1>}
 dyn_strap_l_01 {parent=dyn_strap_l_base, loc=<0 0 0>, rot=<0 0 0 1>}
 clavicle_r {parent=spine_05, loc=<5.4831 -0.1053 3.2946>, rot=<0.7289 0.0465 -0.6829 0.009>}
 upperarm_r {parent=clavicle_r, loc=<-9.8658 1.2136 0.4105>, rot=<0.0164 -0.3245 -0.0269 -0.9454>}
 lowerarm_r {parent=upperarm_r, loc=<-26.465 0.8345 -0.259>, rot=<0.0036 0.0161 -0.2298 -0.9731>}
 hand_r {parent=lowerarm_r, loc=<-22.7355 0.3203 0.6106>, rot=<-0.6784 -0.0223 0.0735 -0.7307>}
 thumb_01_r {parent=hand_r, loc=<-1.3118 0.3191 2.459>, rot=<0.6546 -0.4042 0.0664 -0.6354>}
 thumb_02_r {parent=thumb_01_r, loc=<-5.9336 0.0448 -0.0401>, rot=<-0.0037 0.0213 0.298 -0.9543>}
 thumb_03_r {parent=thumb_02_r, loc=<-2.2223 -0.3682 0.0907>, rot=<0.0118 -0.013 0.0066 -0.9998>}
 index_metacarpal_r {parent=hand_r, loc=<-1.6535 -0.0183 2.96>, rot=<0.0158 -0.0124 -0.0173 -0.9996>}
 index_01_r {parent=index_metacarpal_r, loc=<-8.3008 -0.0357 0.0524>, rot=<-0.0044 0.0485 0.0924 -0.9945>}
 index_02_r {parent=index_01_r, loc=<-3.3587 0.3004 -0.0452>, rot=<-0.002 0.0027 0.3225 -0.9466>}
 index_03_r {parent=index_02_r, loc=<-2.5578 0.123 0.0252>, rot=<-0.002 0.0067 0.2558 -0.9667>}
 middle_metacarpal_r {parent=hand_r, loc=<-2.1856 0.1725 0.5984>, rot=<0.006 0.0168 -0.0329 -0.9993>}
 middle_01_r {parent=middle_metacarpal_r, loc=<-7.3275 0.2447 -0.0129>, rot=<-0.0208 0.0042 0.0858 -0.9961>}
 middle_02_r {parent=middle_01_r, loc=<-4.0123 -0.1401 -0.1175>, rot=<-0.0218 0.0258 0.3512 -0.9357>}
 middle_03_r {parent=middle_02_r, loc=<-2.6077 -0.0534 -0.1373>, rot=<0.0082 0.0188 0.1708 -0.9851>}
 ring_metacarpal_r {parent=hand_r, loc=<-2.2074 0.3625 -1.3638>, rot=<-0.0284 0.0722 -0.0458 -0.9959>}
 ring_01_r {parent=ring_metacarpal_r, loc=<-7.0295 0.6197 0.1446>, rot=<-0.0489 -0.0252 0.1892 -0.9804>}
 ring_02_r {parent=ring_01_r, loc=<-3.6174 -0.4647 0.1027>, rot=<0 0 0.2806 -0.9598>}
 ring_03_r {parent=ring_02_r, loc=<-2.4152 -0.114 -0.0199>, rot=<0 0 0.1189 -0.9929>}
 pinky_metacarpal_r {parent=hand_r, loc=<-1.7616 0.9753 -3.3814>, rot=<-0.0397 0.0814 -0.069 -0.9935>}
 pinky_01_r {parent=pinky_metacarpal_r, loc=<-7.1709 0.4831 0.1094>, rot=<-0.0654 -0.0035 0.2242 -0.9723>}
 pinky_02_r {parent=pinky_01_r, loc=<-2.8274 -0.0893 0.1043>, rot=<0.0037 0.0011 0.2854 -0.9584>}
 pinky_03_r {parent=pinky_02_r, loc=<-2.2046 -0.0307 0.0078>, rot=<-0.0002 0.0098 0.1445 -0.9895>}
 deform_wrist_dn_base_r {parent=hand_r, loc=<-0.0001 1.5328 0>, rot=<0 0 0 -1>}
 deform_wrist_dn_r {parent=deform_wrist_dn_base_r, loc=<0 0 0>, rot=<0 0 0 1>}
 deform_wrist_up_base_r {parent=hand_r, loc=<-0.0002 -2.3007 -0.0001>, rot=<0 0 0 -1>}
 deform_wrist_up_r {parent=deform_wrist_up_base_r, loc=<0 0 0>, rot=<0 0 0 1>}
 lowerarm_twist_02_r {parent=lowerarm_r, loc=<-10.6707 0 0>, rot=<0 0 0 -1>}
 lowerarm_twist_01_r {parent=lowerarm_r, loc=<-18.316 0 0>, rot=<0 0 0 -1>}
 deform_elbow_in_base_r {parent=lowerarm_r, loc=<-0.1834 -3.6062 0.2572>, rot=<0 0 0 -1>}
 deform_elbow_in_r {parent=deform_elbow_in_base_r, loc=<0 0 0>, rot=<0 0 0 1>}
 deform_elbow_out_base_r {parent=lowerarm_r, loc=<0 4 -0.0001>, rot=<0 0 0 -1>}
 deform_elbow_out_r {parent=deform_elbow_out_base_r, loc=<0 0 0>, rot=<0 0 0 1>}
 upperarm_twist_01_r {parent=upperarm_r, loc=<-7.2135 0.1134 0>, rot=<0 0 0 -1>}
 deform_upperarm_bk_base_r {parent=upperarm_twist_01_r, loc=<6.4053 4.3746 -0.314>, rot=<0 0 0 -1>}
 deform_upperarm_bk_r {parent=deform_upperarm_bk_base_r, loc=<0 0 0>, rot=<0 0 0 1>}
 deform_upperarm_fr_base_r {parent=upperarm_twist_01_r, loc=<5.6312 -5.092 -0.3354>, rot=<0 0 0 -1>}
 deform_upperarm_fr_r {parent=deform_upperarm_fr_base_r, loc=<0 0 0>, rot=<0 0 0 1>}
 deform_upperarm_up_base_r {parent=upperarm_twist_01_r, loc=<8.891 -0.4825 -3.9517>, rot=<0 0 0 -1>}
 deform_upperarm_up_r {parent=deform_upperarm_up_base_r, loc=<0 0 0>, rot=<0 0 0 1>}
 upperarm_twist_02_r {parent=upperarm_r, loc=<-19.7998 0.1134 0>, rot=<0 0 0 -1>}
 dyn_thorn_r_base {parent=upperarm_r, loc=<0.2975 -0.1824 -4.0373>, rot=<0.4087 -0.0074 0.9126 -0.0019>}
 dyn_thorn_r_01 {parent=dyn_thorn_r_base, loc=<0 0 0>, rot=<0 0 0 1>}
 dyn_strap_r_base {parent=clavicle_r, loc=<-5.783 1.1776 -4.3803>, rot=<0 0 0 1>}
 dyn_strap_r_01 {parent=dyn_strap_r_base, loc=<0 0 0>, rot=<0 0 0 1>}
 attach_backpack {parent=spine_05, loc=<-0.123 -13.473 0>, rot=<-0.0093 -0.7069 0.0093 -0.7071>}
 attach_cape {parent=spine_05, loc=<9.017 -11.714 0>, rot=<0.0616 -0.7043 -0.0616 -0.7045>}
 dyn_thorn_2_r_base {parent=spine_05, loc=<9.5657 -4.6477 7.772>, rot=<0.0401 -0.9992 0 0>}
 dyn_thorn_2_r_01 {parent=dyn_thorn_2_r_base, loc=<0 0 0>, rot=<0 0 0 1>}
 dyn_thorn_1_r_base {parent=spine_05, loc=<10.0738 -2.4878 8.8735>, rot=<0.0401 -0.9992 0 0>}
 dyn_thorn_1_r_01 {parent=dyn_thorn_1_r_base, loc=<0 0 0>, rot=<0 0 0 1>}
 dyn_thorn_l_base {parent=spine_05, loc=<9.5657 -4.6477 -7.772>, rot=<0.0401 -0.9992 0 0>}
 dyn_thorn_l_01 {parent=dyn_thorn_l_base, loc=<0 0 0>, rot=<0 0 0 1>}
 dyn_thorn_1_l_base {parent=spine_05, loc=<10.0738 -2.4878 -8.8735>, rot=<0.0401 -0.9992 0 0>}
 dyn_thorn_1_l_01 {parent=dyn_thorn_1_l_base, loc=<0 0 0>, rot=<0 0 0 1>}
 deform_pec_aim_base_l {parent=spine_04, loc=<3.2456 9.3309 -1.2237>, rot=<0.0603 -0.4033 0.3318 -0.8507>}
 deform_pec_aim_l {parent=deform_pec_aim_base_l, loc=<0 0 0>, rot=<0 0 0 1>}
 deform_pec_base_l {parent=deform_pec_aim_l, loc=<9.3665 0 0>, rot=<0 0 0 1>}
 deform_pec_l {parent=deform_pec_base_l, loc=<0 0 0>, rot=<0 0 0 1>}
 deform_pec_aim_base_r {parent=spine_04, loc=<3.2452 9.3309 1.2237>, rot=<-0.4033 -0.0603 0.8507 0.3318>}
 deform_pec_aim_r {parent=deform_pec_aim_base_r, loc=<0 0 0>, rot=<0 0 0 1>}
 deform_pec_base_r {parent=deform_pec_aim_r, loc=<-9.3667 -0.0003 -0.0004>, rot=<0 0 0 1>}
 deform_pec_r {parent=deform_pec_base_r, loc=<0 0 0>, rot=<0 0 0 1>}
 thigh_l {parent=pelvis, loc=<-1.0855 -0.2109 -8.7053>, rot=<0.0009 0.0355 0.0155 -0.9993>}
 calf_l {parent=thigh_l, loc=<-42.5284 -0.0019 -0.7297>, rot=<0.0006 -0.0161 -0.0391 -0.9991>}
 foot_l {parent=calf_l, loc=<-46.7401 -2.5014 -0.592>, rot=<0.0427 -0.0238 0.0353 -0.9982>}
 ball_l {parent=foot_l, loc=<-3.5658 10.8066 -0.1544>, rot=<0.0001 -0.0001 -0.7063 -0.7079>}
 support_foot_l {parent=foot_l, loc=<3.4999 -5.9999 0>, rot=<0 0 0 1>}
 calf_twist_02_l {parent=calf_l, loc=<-19.1546 -1.7943 -0.2638>, rot=<0.0237 0.0095 -0.0035 -0.9997>}
 calf_twist_01_l {parent=calf_l, loc=<-35.8483 -1.9197 -0.5794>, rot=<0.0237 0.0095 -0.0035 -0.9997>}
 deform_knee_in_base_l {parent=calf_l, loc=<0.8031 -5.0231 0.2456>, rot=<0 0 0 -1>}
 deform_knee_in_l {parent=deform_knee_in_base_l, loc=<0 0 0>, rot=<0 0 0 1>}
 deform_knee_out_base_l {parent=calf_l, loc=<0 4 0>, rot=<0 0 0 -1>}
 deform_knee_out_l {parent=deform_knee_out_base_l, loc=<0 0 0>, rot=<0 0 0 1>}
 thigh_twist_01_l {parent=thigh_l, loc=<-27.952 0 0>, rot=<0.0013 0 0 -1>}
 deform_groin_base_l {parent=thigh_l, loc=<0.4801 5.6776 -0.2486>, rot=<0 0 0 -1>}
 deform_groin_l {parent=deform_groin_base_l, loc=<0 0 0>, rot=<0 0 0 1>}
 thigh_r {parent=pelvis, loc=<-1.0855 -0.2109 8.7053>, rot=<0.0355 -0.0009 0.9993 0.0155>}
 calf_r {parent=thigh_r, loc=<42.5284 0.0019 0.7297>, rot=<0.0006 -0.0161 -0.0391 -0.9991>}
 foot_r {parent=calf_r, loc=<46.7401 2.5014 0.592>, rot=<0.0427 -0.0238 0.0353 -0.9982>}
 ball_r {parent=foot_r, loc=<3.5658 -10.8066 0.1544>, rot=<0.0001 -0.0001 -0.7063 -0.7079>}
 support_foot_r {parent=foot_r, loc=<-3.4999 5.9997 0.0001>, rot=<0 0 0 1>}
 calf_twist_02_r {parent=calf_r, loc=<19.1546 1.7943 0.2638>, rot=<0.0237 0.0095 -0.0035 -0.9997>}
 calf_twist_01_r {parent=calf_r, loc=<35.8483 1.9197 0.5794>, rot=<0.0237 0.0095 -0.0035 -0.9997>}
 deform_knee_in_base_r {parent=calf_r, loc=<-0.8031 5.0231 -0.2457>, rot=<0 0 0 -1>}
 deform_knee_in_r {parent=deform_knee_in_base_r, loc=<0 0 0>, rot=<0 0 0 1>}
 deform_knee_out_base_r {parent=calf_r, loc=<0 -4 0>, rot=<0 0 0 -1>}
 deform_knee_out_r {parent=deform_knee_out_base_r, loc=<0 0 0>, rot=<0 0 0 1>}
 thigh_twist_01_r {parent=thigh_r, loc=<27.952 0 0>, rot=<0.0013 0 0 -1>}
 deform_groin_base_r {parent=thigh_r, loc=<-0.4801 -5.6776 0.2486>, rot=<0 0 0 -1>}
 deform_groin_r {parent=deform_groin_base_r, loc=<0 0 0>, rot=<0 0 0 1>}
 deform_glute_base_l {parent=pelvis, loc=<-7.3466 -9.0213 -9.1245>, rot=<0.0172 0.031 -0.4477 -0.8935>}
 deform_glute_l {parent=deform_glute_base_l, loc=<0 0 0>, rot=<0 0 0 1>}
 deform_glute_base_r {parent=pelvis, loc=<-7.3464 -9.0213 9.1245>, rot=<0.031 -0.0172 0.8935 -0.4477>}
 deform_glute_r {parent=deform_glute_base_r, loc=<0 0 0>, rot=<0 0 0 1>}
 base_knives1 {parent=pelvis, loc=<8.3757 -2.6885 10.6415>, rot=<0.0956 -0.2171 0.971 -0.0292>}
]
variables [
 B {type=double, default=1}
 C {type=double, default=1}
 is Aim {type=bool, default=False}
 is pistol {type=bool, default=False}
 is rifle {type=bool, default=False}
]
