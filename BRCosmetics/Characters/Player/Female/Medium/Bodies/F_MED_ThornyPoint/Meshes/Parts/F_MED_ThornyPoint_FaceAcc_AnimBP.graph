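# Animation Blueprint for the ThornyPoint female character's FaceAcc part, posing it from the character and simulating its secondary motion.
# It blends the attached parent's pose with arm and leg bones copied from the owner body mesh, then offsets one front hair bone.
# A Control Rig reacts to targeting and pistol or rifle animation states.
# A head-relative RigidBody simulation follows, using per-update gravity from a nativized state function and parameter asset.
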AnimBlueprint /BRCosmetics/Characters/Player/Female/Medium/Bodies/F_MED_ThornyPoint/Meshes/Parts/F_MED_ThornyPoint_FaceAcc_AnimBP
target_skeleton=&/F_MED_ThornyPoint_FaceAcc_Skeleton
graphs [
 AnimGraph {nodes [
   n1 = Layered blend per bone(
    BasePose=Copy Pose From Mesh() {bUseAttachedParent=true}
    BlendPoses_0=Copy Pose From Mesh(SourceMeshComponent=Get body() -> (body:Skeletal Mesh Component Object Reference)) {bUseAttachedParent=true}
   ) {layerSetup [{branchFilters [
       {boneName=upperarm_l}
       {boneName=upperarm_r}
       {boneName=lowerarm_l}
       {boneName=lowerarm_r}
       {boneName=thigh_l}
       {boneName=thigh_r}
       {boneName=calf_l}
       {boneName=calf_r}
      ]}]}
   n2 = "Transform (Modify) Bone - Bone: dyn_hair_fr_r_base2"(ComponentPose=Local To Component(LocalPose=@n1), Alpha=1, Translation="-1.5,0,0", Rotation="0,0,0", Scale="1,1,1") {rotationMode=BMM_Additive, translationMode=BMM_Additive, boneToModify {boneName=dyn_hair_fr_r_base2}, translation {x=-1.5}}
   n3 = Control Rig!(
    Source=Component To Local(ComponentPose=@n2)
    Alpha
    is Aim:bool=Get bIsTargeting()
    is pistol:bool=Get pistol() -> (pistol:bool)
    is rifle:bool=Get rifle() -> (rifle:bool)
   ) -> (Pose) {controlRigAssetReference {blueprintRigClass=&/F_MED_ThornyPoint_FaceAcc_CtrlRig.F_MED_ThornyPoint_FaceAcc_CtrlRig_C}}
   n4 = RigidBody!(ComponentPose=Local To Component(LocalPose=@n3), Alpha, OverrideWorldGravity:Vector=Get Gravity Override(), ExternalForce) -> (Pose) {
    bClampLinearTranslationLimitToRefPose=true
    bForceDisableCollisionBetweenConstraintBodies=true
    bOverrideWorldGravity=true
    overridePhysicsAsset=&/F_MED_ThornyPoint_FaceAcc_Physics
    simulationSpace=BaseBoneSpace
    alphaBoolBlend {blendOption=Linear}
    baseBoneRef {boneName=head}
    simSpaceSettings {
     dampingAlpha=0
     maxAngularAcceleration=40
     maxAngularVelocity=6
     maxLinearAcceleration=1500
     maxLinearVelocity=700
     velocityScaleZ=0.7
     worldAlpha=1
     externalLinearDragV {x=2}
    }
   }
   Output Pose(Result=Component To Local(ComponentPose=@n4)) {layerGroup=None}
  ]}
 EventGraph {nodes [
   Event BlueprintUpdateAnimation()
   > n1 = Sequence()
   > n2 = "Gravity Override State Function (Parameter Asset) (Nativized)"(
    CharacterPartAnimInstance=Self-Reference()
    GameTimeInSeconds=GetGameTimeInSeconds()
    Parameters=/BRCosmetics/Characters/Player/Female/Medium/Bodies/F_MED_ThornyPoint/Meshes/M_MED_ThornyPoint_Parameters
   )
   > Set Gravity Override(Gravity Override=@n2)
   n3 = Get WeaponCoreAnim!() -> (WeaponCoreAnim)
   Set pistol(execute=@n1.then_1, pistol:bool="Equal (Enum)"(A:EFortWeaponCoreAnimation Enum=@n3, B:EFortWeaponCoreAnimation Enum=Pistol)) -> (Output_Get:bool)
   > Set rifle(rifle:bool="Equal (Enum)"(A=@n3, B=AssaultRifle)) -> (Output_Get:bool)
   Event OnCharacterCustomizationLoaded()
   > Set body(body:Skeletal Mesh Component Object Reference=GetOwnerSkeletalMeshComponent()) -> (Output_Get:Skeletal Mesh Component Object Reference)
  ]}
]
variables [
 body {type=SkeletalMeshComponent}
 Gravity Override {type=Vector}
 pistol {type=boolean}
 rifle {type=boolean}
]
