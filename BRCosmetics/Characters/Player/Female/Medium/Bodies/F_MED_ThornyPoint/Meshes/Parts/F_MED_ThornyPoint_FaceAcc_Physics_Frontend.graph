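# Frontend physics asset for the F_MED_ThornyPoint face accessory, simulating dynamic hair and chain bone strands around the head.
# Capsules and spheres on head, neck, upper torso, shoulders and upper arms form the body shapes, joined by locked constraints.
# Front, side and back strands hang from head-locked base bones through per-joint swing and twist limits.
# Extra angular-free constraints cross-link the center back chain to both back side strands and neck_02.
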PhysicsAsset /BRCosmetics/Characters/Player/Female/Medium/Bodies/F_MED_ThornyPoint/Meshes/Parts/F_MED_ThornyPoint_FaceAcc_Physics_Frontend
bodies {
 clavicle_l [{shapeType=Capsule, length=9.1882, radius=3.364, shapeName=clavicle_l_capsule, center=<5.3918 0 -0.5237>, rotation=<rot 0 -90 95.5477>}]
 clavicle_r [{
   shapeType=Capsule
   length=9.6831
   radius=3.5292
   shapeName=clavicle_r_capsule
   center=<-5.3853 0.0953 0.5231>
   rotation=<rot -0.1865 -88.0805 -84.4554>
  }]
 dyn_chain_back_01 [{
   shapeType=Capsule
   length=6.1874
   radius=2.8111
   shapeName=dyn_chain_back_01_capsule
   center=<3.8524 0 -0.6922>
   rotation=<rot 1.2215 -0.7481 -88.4509>
  }]
 dyn_chain_back_02 [{
   shapeType=Capsule
   length=6.7708
   radius=3.4319
   shapeName=dyn_chain_back_02_capsule
   center=<3.2347 0 0.2017>
   rotation=<rot 11.2338 1.9973 -80.5837>
  }]
 dyn_chain_back_03 [{
   shapeType=Capsule
   length=6.3858
   radius=3.729
   shapeName=dyn_chain_back_03_capsule
   center=<3.1162 0 0.83>
   rotation=<rot 0 0.4106 -87.8885>
  }]
 dyn_chain_back_base [{shapeType=Capsule, length=0.505, radius=0.505, shapeName=dyn_chain_back_base_capsule, rotation=<rot 0 -90 0>}]
 dyn_chain_bk_r_01 [{
   shapeType=Capsule
   length=5.5095
   radius=2.068
   shapeName=dyn_chain_bk_r_01_capsule
   center=<3.5958 -1.468 -0.7408>
   rotation=<rot -10.7827 -172.3761 -77.5763>
  }]
 dyn_chain_bk_r_02 [{
   shapeType=Capsule
   length=5.6942
   radius=2.2311
   shapeName=dyn_chain_bk_r_02_capsule
   center=<2.8844 -1.2025 -0.9892>
   rotation=<rot -4.351 -174.7846 -85.9059>
  }]
 dyn_chain_bk_r_03 [{
   shapeType=Capsule
   length=3.4014
   radius=1.7092
   shapeName=dyn_chain_bk_r_03_capsule
   center=<2.3043 -1.3801 -0.9107>
   rotation=<rot 0.0001 172.9346 -74.4765>
  }]
 dyn_chain_bk_r_base [{shapeType=Capsule, length=0.505, radius=0.505, shapeName=dyn_chain_bk_r_base_capsule, rotation=<rot 0 -90 0>}]
 dyn_chain_r_01 [
  {
   shapeType=Capsule
   length=5.3563
   radius=1.7494
   shapeName=dyn_chain_r_01_capsule
   center=<5.1153 -1.2858 -0.7917>
   rotation=<rot 3.1803 -146.5503 -82.1378>
  }
  {
   shapeType=Capsule
   length=4.0768
   radius=1.7118
   shapeName=dyn_chain_r_01_capsule
   center=<7.8662 -1.2422 -1.5145>
   rotation=<rot 4.0863 32.7601 -102.4254>
  }
 ]
 dyn_chain_r_02 [{
   shapeType=Capsule
   length=5.0225
   radius=2.0008
   shapeName=dyn_chain_r_02_capsule
   center=<7.2524 -3.0698 -2.1113>
   rotation=<rot 1.4812 -136.8879 -84.0954>
  }]
 dyn_chain_r_base [{shapeType=Capsule, length=0.505, radius=0.505, shapeName=dyn_chain_r_base_capsule, rotation=<rot 0 -90 0>}]
 dyn_hair_bk_l_01 [{
   shapeType=Capsule
   length=6.7957
   radius=1.9906
   shapeName=dyn_hair_bk_l_01_capsule
   center=<3.8388 -0.1514 -0.7106>
   rotation=<rot -2.0385 -8.9886 -81.9644>
  }]
 dyn_hair_bk_l_02 [{
   shapeType=Capsule
   length=4.3521
   radius=2.0828
   shapeName=dyn_hair_bk_l_02_capsule
   center=<2.7944 -0.5067 -0.0755>
   rotation=<rot -4.5302 -7.3505 -85.094>
  }]
 dyn_hair_bk_l_03 [{
   shapeType=Capsule
   length=3.5544
   radius=2.9527
   shapeName=dyn_hair_bk_l_03_capsule
   center=<2.256 -0.8328 -0.6733>
   rotation=<rot 0.8421 178.96 -83.7792>
  }]
 dyn_hair_bk_l_base [{shapeType=Capsule, length=0.505, radius=0.505, shapeName=dyn_hair_bk_l_base_capsule, rotation=<rot 0 -90 0>}]
 dyn_hair_fr_l_03 [{shapeType=Capsule, length=3.8505, radius=1.3195, shapeName=dyn_hair_02_capsule, center=<2.1598 0 0>, rotation=<rot -90 0.7249 0>}]
 dyn_hair_fr_l_04 [{
   shapeType=Capsule
   length=3.323
   radius=1.4094
   shapeName=dyn_hair_03_capsule
   center=<2.0605 -0.3436 -0.5938>
   rotation=<rot 61.4371 -23.3056 -3.7054>
  }]
 dyn_hair_fr_l_05 [{shapeType=Capsule, length=3.9489, radius=1.5741, shapeName=dyn_hair_01_capsule, center=<3.5478 0 -0.0779>, rotation=<rot -90 176.8934 0>}]
 dyn_hair_fr_l_base2 [{shapeType=Capsule, length=0.505, radius=0.505, shapeName=dyn_hair_fr_l_base2_capsule, rotation=<rot 0 -90 0>}]
 dyn_hair_fr_r_02 [{shapeType=Capsule, length=5.1235, radius=1.5539, shapeName=dyn_hair_01_capsule, center=<3.5531 0.6726 0>, rotation=<rot -90 179.5538 0>}]
 dyn_hair_fr_r_03 [{shapeType=Capsule, length=6.0031, radius=1.7177, shapeName=dyn_hair_02_capsule, center=<2.1285 0.8087 0>, rotation=<rot -90 2.2191 0>}]
 dyn_hair_fr_r_04 [{
   shapeType=Capsule
   length=3.9421
   radius=1.4892
   shapeName=dyn_hair_03_capsule
   center=<1.8742 1.3193 0>
   rotation=<rot 38.0372 179.9079 179.9998>
  }]
 dyn_hair_fr_r_base2 [{shapeType=Capsule, length=0.5014, radius=0.5, shapeName=dyn_hair_fr_r_base2_capsule, rotation=<rot 0 -90 0>}]
 dyn_hair_l_01 [{
   shapeType=Capsule
   length=7.8983
   radius=2.0939
   shapeName=dyn_hair_l_01_capsule
   center=<7.9687 1.3101 -0.6741>
   rotation=<rot -3.9494 -39.1671 -72.574>
  }]
 dyn_hair_l_02 [{
   shapeType=Capsule
   length=7.7332
   radius=1.5797
   shapeName=dyn_hair_l_02_capsule
   center=<7.2827 2.9162 -2.0661>
   rotation=<rot -3.4482 -42.8813 -69.3692>
  }]
 dyn_hair_l_base [{shapeType=Capsule, length=0.505, radius=0.505, shapeName=dyn_hair_l_base_capsule, rotation=<rot 0 -90 0>}]
 head [
  {shapeType=Sphere, radius=2.5207, shapeName=head_sphere, center=<-1.219 8.6518 0>}
  {shapeType=Sphere, radius=2.0059, shapeName=head_sphere, center=<-4.8084 8.4464 0>}
  {shapeType=Capsule, length=13.1581, radius=5.3348, center=<2.3673 0.9769 -0.3126>, rotation=<rot 0.4367 -84.6879 -89.7153>}
  {shapeType=Capsule, length=0.0353, radius=0.5, shapeName=head_capsule, center=<1.6134 1.3222 -0.0239>, rotation=<rot -89.1519 179.9981 -180>}
 ]
 neck_01 [{
   shapeType=Capsule
   length=4.6107
   radius=4.642
   shapeName=neck_01_capsule
   center=<-0.2574 0.99 -0.05>
   rotation=<rot 89.2849 21.8492 27.9774>
  }]
 neck_02 [{shapeType=Capsule, length=2.8177, radius=4.1811, shapeName=neck_02_capsule, rotation=<rot 0 -90 0>}]
 spine_01 [{shapeType=Capsule, length=0.505, radius=0.505, shapeName=spine_01_capsule, rotation=<rot 0 -90 0>}]
 spine_05 [
  {shapeType=Capsule, length=3.5738, radius=7.736, center=<3.5967 0.3006 0.0491>, rotation=<rot -0.5275 -168.227 -179.6905>}
  {shapeType=Capsule, length=22.5304, radius=7.4696, shapeName=spine_05_capsule, center=<-1.7971 2.8371 0>}
  {shapeType=Capsule, length=8.7798, radius=4.1223, shapeName=spine_05_capsule, center=<3.747 3.0971 7.8008>}
 ]
 upperarm_l [{
   shapeType=Capsule
   length=14.9628
   radius=3.382
   shapeName=upperarm_l_capsule
   center=<7.9141 0.0009 -1.1421>
   rotation=<rot -0.0065 -89.9942 98.212>
  }]
 upperarm_r [
  {
   shapeType=Capsule
   length=23.9829
   radius=3.9017
   shapeName=upperarm_r_capsule
   center=<-14.007 2.0315 -0.1142>
   rotation=<rot -0.0065 -89.9942 -81.788>
  }
  {
   shapeType=Capsule
   length=12.3081
   radius=5.2977
   shapeName=upperarm_r_capsule
   center=<-31.4598 2.0686 -2.0685>
   rotation=<rot -83.7403 179.9996 -179.9995>
  }
  {
   shapeType=Capsule
   length=14.6198
   radius=3.7203
   shapeName=upperarm_r_capsule
   center=<-14.8785 7.6392 -4.2268>
   rotation=<rot -83.8757 179.9996 -179.9996>
  }
 ]
}
constraints [
 {
  bone1Name=spine_05
  bone2Name=spine_01
  swing1LimitDegrees=45
  swing1Motion=Locked
  swing2LimitDegrees=45
  swing2Motion=Locked
  twistLimitDegrees=45
  twistMotion=Locked
 }
 {
  bone1Name=neck_01
  bone2Name=spine_05
  swing1LimitDegrees=45
  swing1Motion=Locked
  swing2LimitDegrees=45
  swing2Motion=Locked
  twistLimitDegrees=45
  twistMotion=Locked
 }
 {
  bone1Name=neck_02
  bone2Name=neck_01
  swing1LimitDegrees=45
  swing1Motion=Locked
  swing2LimitDegrees=45
  swing2Motion=Locked
  twistLimitDegrees=45
  twistMotion=Locked
 }
 {
  bone1Name=head
  bone2Name=neck_02
  swing1LimitDegrees=45
  swing1Motion=Locked
  swing2LimitDegrees=45
  swing2Motion=Locked
  twistLimitDegrees=45
  twistMotion=Locked
 }
 {
  bone1Name=dyn_hair_fr_l_base2
  bone2Name=head
  swing1LimitDegrees=45
  swing1Motion=Locked
  swing2LimitDegrees=45
  swing2Motion=Locked
  twistLimitDegrees=45
  twistMotion=Locked
 }
 {bone1Name=dyn_hair_fr_l_05, bone2Name=dyn_hair_fr_l_base2, swing1LimitDegrees=16.76, swing2LimitDegrees=18.2, twistLimitDegrees=4}
 {bone1Name=dyn_hair_fr_l_03, bone2Name=dyn_hair_fr_l_05, swing1LimitDegrees=22, swing2LimitDegrees=22, twistLimitDegrees=3}
 {bone1Name=dyn_hair_fr_l_04, bone2Name=dyn_hair_fr_l_03, swing1LimitDegrees=22, swing2LimitDegrees=22, twistLimitDegrees=3}
 {
  bone1Name=dyn_hair_fr_r_base2
  bone2Name=head
  swing1LimitDegrees=45
  swing1Motion=Locked
  swing2LimitDegrees=45
  swing2Motion=Locked
  twistLimitDegrees=45
  twistMotion=Locked
 }
 {bone1Name=dyn_hair_fr_r_02, bone2Name=dyn_hair_fr_r_base2, swing1LimitDegrees=22, swing2LimitDegrees=22, twistLimitDegrees=11}
 {bone1Name=dyn_hair_fr_r_03, bone2Name=dyn_hair_fr_r_02, swing1LimitDegrees=41.584, swing2LimitDegrees=36.976, twistLimitDegrees=3}
 {bone1Name=dyn_hair_fr_r_04, bone2Name=dyn_hair_fr_r_03, swing1LimitDegrees=46.192, swing2LimitDegrees=48.888, twistLimitDegrees=3}
 {
  bone1Name=clavicle_l
  bone2Name=spine_05
  swing1LimitDegrees=45
  swing1Motion=Locked
  swing2LimitDegrees=45
  swing2Motion=Locked
  twistLimitDegrees=45
  twistMotion=Locked
 }
 {
  bone1Name=upperarm_l
  bone2Name=clavicle_l
  swing1LimitDegrees=45
  swing1Motion=Locked
  swing2LimitDegrees=45
  swing2Motion=Locked
  twistLimitDegrees=45
  twistMotion=Locked
 }
 {
  bone1Name=clavicle_r
  bone2Name=spine_05
  swing1LimitDegrees=45
  swing1Motion=Locked
  swing2LimitDegrees=45
  swing2Motion=Locked
  twistLimitDegrees=45
  twistMotion=Locked
 }
 {
  bone1Name=upperarm_r
  bone2Name=clavicle_r
  swing1LimitDegrees=45
  swing1Motion=Locked
  swing2LimitDegrees=45
  swing2Motion=Locked
  twistLimitDegrees=45
  twistMotion=Locked
 }
 {
  bone1Name=dyn_chain_r_base
  bone2Name=head
  swing1LimitDegrees=45
  swing1Motion=Locked
  swing2LimitDegrees=45
  swing2Motion=Locked
  twistLimitDegrees=45
  twistMotion=Locked
 }
 {bone1Name=dyn_chain_r_01, bone2Name=dyn_chain_r_base, swing1LimitDegrees=25.6, swing2LimitDegrees=13.824, twistLimitDegrees=11}
 {bone1Name=dyn_chain_r_02, bone2Name=dyn_chain_r_01, swing1LimitDegrees=25.4, swing2LimitDegrees=34.4, twistLimitDegrees=11}
 {
  bone1Name=dyn_hair_l_base
  bone2Name=head
  swing1LimitDegrees=45
  swing1Motion=Locked
  swing2LimitDegrees=45
  swing2Motion=Locked
  twistLimitDegrees=45
  twistMotion=Locked
 }
 {bone1Name=dyn_hair_l_01, bone2Name=dyn_hair_l_base, swing1LimitDegrees=22, swing2LimitDegrees=11, twistLimitDegrees=3}
 {bone1Name=dyn_hair_l_02, bone2Name=dyn_hair_l_01, swing1LimitDegrees=22, swing2LimitDegrees=11, twistLimitDegrees=4}
 {bone1Name=dyn_hair_bk_l_03, bone2Name=dyn_hair_bk_l_02, swing1LimitDegrees=20.608, swing2LimitDegrees=36.736, twistLimitDegrees=11}
 {bone1Name=dyn_hair_bk_l_01, bone2Name=dyn_hair_bk_l_base, swing1LimitDegrees=15.328, swing2LimitDegrees=13.824, twistLimitDegrees=11}
 {bone1Name=dyn_hair_bk_l_02, bone2Name=dyn_hair_bk_l_01, swing1LimitDegrees=15.608, swing2LimitDegrees=13.824, twistLimitDegrees=11}
 {
  bone1Name=dyn_hair_bk_l_base
  bone2Name=head
  swing1LimitDegrees=45
  swing1Motion=Locked
  swing2LimitDegrees=45
  swing2Motion=Locked
  twistLimitDegrees=45
  twistMotion=Locked
 }
 {bone1Name=dyn_chain_bk_r_03, bone2Name=dyn_chain_bk_r_02, swing1LimitDegrees=11, swing2LimitDegrees=24, twistLimitDegrees=35.192}
 {
  bone1Name=dyn_chain_bk_r_base
  bone2Name=head
  swing1LimitDegrees=45
  swing1Motion=Locked
  swing2LimitDegrees=45
  swing2Motion=Locked
  twistLimitDegrees=45
  twistMotion=Locked
 }
 {bone1Name=dyn_chain_bk_r_01, bone2Name=dyn_chain_bk_r_base, swing1LimitDegrees=19.328, swing2LimitDegrees=8.96, twistLimitDegrees=4}
 {bone1Name=dyn_chain_bk_r_02, bone2Name=dyn_chain_bk_r_01, swing1LimitDegrees=25.4, swing2LimitDegrees=10.28, twistLimitDegrees=11}
 {bone1Name=dyn_chain_back_03, bone2Name=dyn_chain_back_02, swing1LimitDegrees=5, swing2LimitDegrees=25.976, twistLimitDegrees=11}
 {
  bone1Name=dyn_chain_back_base
  bone2Name=head
  swing1LimitDegrees=45
  swing1Motion=Locked
  swing2LimitDegrees=45
  swing2Motion=Locked
  twistLimitDegrees=45
  twistMotion=Locked
 }
 {bone1Name=dyn_chain_back_01, bone2Name=dyn_chain_back_base, swing1LimitDegrees=5, swing2LimitDegrees=11, twistLimitDegrees=11}
 {bone1Name=dyn_chain_back_02, bone2Name=dyn_chain_back_01, swing1LimitDegrees=13.3432, swing2LimitDegrees=16.76, twistLimitDegrees=11}
 {
  bone1Name=dyn_hair_bk_l_03
  bone2Name=dyn_chain_back_03
  swing1LimitDegrees=45
  swing1Motion=Free
  swing2LimitDegrees=45
  swing2Motion=Free
  twistLimitDegrees=45
  twistMotion=Free
 }
 {
  bone1Name=dyn_chain_bk_r_03
  bone2Name=dyn_chain_back_03
  swing1LimitDegrees=45
  swing1Motion=Free
  swing2LimitDegrees=45
  swing2Motion=Free
  twistLimitDegrees=45
  twistMotion=Free
 }
 {
  bone1Name=dyn_chain_back_02
  bone2Name=neck_02
  swing1LimitDegrees=45
  swing1Motion=Free
  swing2LimitDegrees=45
  swing2Motion=Free
  twistLimitDegrees=45
  twistMotion=Free
 }
]
